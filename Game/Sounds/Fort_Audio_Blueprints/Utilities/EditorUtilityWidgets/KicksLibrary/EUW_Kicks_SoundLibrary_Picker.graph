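# Editor utility widget that assigns a cosmetic footstep Sound Library to a Kicks (shoe) item's custom character parts.
# On Add Library, the chosen override enum selects a library, which is written to each part's SoundLibraries via editor property calls, then saved.
# An empty selection resets that property instead.
# A timed status box confirms the add or clear, and warning dialogs flag a missing item or parts.
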EditorUtilityWidgetBlueprint /Game/Sounds/Fort_Audio_Blueprints/Utilities/EditorUtilityWidgets/KicksLibrary/EUW_Kicks_SoundLibrary_Picker
graphs [EventGraph {nodes [
   "On Pressed (AddLibrary)"() {is=title K2Node_ComponentBoundEvent}
   > n1 = Get!(execute:Exec) -> (then:Exec, else:Exec, KicksItemDef:Cosmetic Shoes Item Definition Object Reference)
   > Set SoundLibrary(SoundLibrary:Sound Library Object Reference=Select!(
     NewEnumerator8:Sound Library Object Reference
     NewEnumerator0:Sound Library Object Reference=/Game/Sounds/Fort_Player/Footsteps/Barefoot/SL_Footsteps_Barefoot_Cosmetic
     NewEnumerator1:Sound Library Object Reference=/Game/Sounds/Fort_Player/Footsteps/BarefootHeavy/SL_Footsteps_BarefootHeavy_Cosmetic
     NewEnumerator2:Sound Library Object Reference=/Game/Sounds/Fort_Player/Footsteps/CombatBoots/SL_Footsteps_BootsCombat_Cosmetic
     NewEnumerator3:Sound Library Object Reference=/Game/Sounds/Fort_Player/Footsteps/HeeledBoots/SL_Footsteps_BootsHeeled_Cosmetic
     NewEnumerator4:Sound Library Object Reference=/Game/Sounds/Fort_Player/Footsteps/MetalBoots/SL_Footsteps_BootsMetal_Cosmetic
     NewEnumerator5:Sound Library Object Reference=/Game/Sounds/Fort_Player/Footsteps/Sandals/SL_Footsteps_Sandals_Cosmetics
     NewEnumerator6:Sound Library Object Reference=/Game/Sounds/Fort_Player/Footsteps/DressShoes/SL_Footsteps_ShoesDress_Cosmetic
     NewEnumerator7:Sound Library Object Reference=/Game/Sounds/Fort_Player/Footsteps/Sneakers/SL_Footsteps_ShoesSneakers_Cosmetic
     Index:E_KicksAudioOverrideType Enum=Get KicksOverrideType() -> (KicksOverrideType:E_KicksAudioOverrideType Enum=NewEnumerator8)
    ) -> (ReturnValue:Sound Library Object Reference)) -> (Output_Get:Sound Library Object Reference)
   > n2 = GetEditorProperty!(execute:Exec, Object:Object Reference=@n1, PropertyName:Name=CharacterParts) -> (then:Exec, PropertyValue:Array of Custom Character Part Object References, ReturnValue:bool)
   > n3 = Set CharacterParts(CharacterParts:Array of Custom Character Part Object References=@n2) -> (Output_Get:Array of Custom Character Part Object References)
   > n4 = Branch(Condition=@n2.ReturnValue)
   > Clear and Invalidate Timer by Handle!(execute, WorldContextObject:Object Reference, Handle=Get Timer() -> (Timer:Timer Handle Structure)) -> (then)
   > n5 = For Each Loop(Array:Array of Custom Character Part Object References=@n3) -> (Array Element:Custom Character Part Object Reference)
   n6 = Get!(execute=@n5.Completed) -> (then, else, SoundLibrary:Sound Library Object Reference)
   > SetVisibility(self=Get SlDisplay() -> (SlDisplay:Vertical Box Object Reference), InVisibility=Visible)
   > Set Text(Text="LOCTEXT(\"Sound Library\\r\\nAdded\")", self=Get FinishedText() -> (FinishedText:Text Object Reference))
   n7 = Get SoundLibrary() -> (SoundLibrary:Sound Library Object Reference)
   > n8 = Set Text(Text="To Text (String)"(InString=GetDisplayName(Object=@n7)), self=Get SLName() -> (SLName:Text Object Reference))
   n9 = SetVisibility(execute=[@n8.then, @n12.then], self=Get FinishedAddingBox() -> (FinishedAddingBox:Horizontal Box Object Reference), InVisibility=Visible)
   n10 = RemoveVisiblity() {is=title K2Node_CustomEvent}
   n11 = Set Timer by Event(execute=@n9.then, Delegate=@n10, Time=5, bLooping=false)
   > Set Timer(Timer:Timer Handle Structure=@n11) -> (Output_Get:Timer Handle Structure)
   SetVisibility(execute=@n6.else, self=Get SlDisplay(), InVisibility=Hidden)
   > n12 = Set Text(Text="LOCTEXT(\"Sound Library\\r\\nCleared\")", self=Get FinishedText())
   n13 = Get(execute=@n5.LoopBody)
   > n14 = SetEditorProperty(
    Object=@n5
    PropertyName=SoundLibraries
    PropertyValue:Array of Sound Library Object References=Make Array("[0]":Sound Library Object Reference=@n13) -> (Array:Array of Sound Library Object References)
   )
   SaveLoadedAsset(execute=[@n15.then, @n14.then], AssetToSave=@n5, bOnlyIfIsDirty=false)
   n15 = ResetEditorProperty!(execute=@n13.else, Object:Object Reference=@n5, PropertyName:Name=SoundLibraries, ChangeNotifyMode:EPropertyAccessChangeNotifyMode Enum) -> (then, ReturnValue:bool)
   NoPartsError(execute=@n4.else)
   NoCIDError(execute=@n1.else)
   NoCIDError() {is=title K2Node_CustomEvent}
   > Show Message Dialog(
    Title=MakeLiteralText(Value="LOCTEXT(\"No Kicks Item delected\")")
    Message=MakeLiteralText(Value="LOCTEXT(\"Ensure you have a valid Shoe Item Definition selected\")")
    MessageType=Ok
    DefaultValue=No
    MessageCategory=Warning
   )
   NoPartsError() {is=title K2Node_CustomEvent}
   > Show Message Dialog(
    Title=MakeLiteralText(Value="LOCTEXT(\"No character part found\")")
    Message=MakeLiteralText(Value="LOCTEXT(\"Ensure your kicks already have a custom character part added.\")")
    MessageType=Ok
    DefaultValue=No
    MessageCategory=Warning
   )
   Event PreConstruct()
   > SetObject(
    self=[
     Get KicksSelection() -> (KicksSelection:Single Property View Object Reference)
     Get SoundLibrarySelection() -> (SoundLibrarySelection:Single Property View Object Reference)
     Get SoundLibraryAdded() -> (SoundLibraryAdded:Single Property View Object Reference)
    ]
    NewObject=Self-Reference()
   )
   Event Destruct()
   > SetVisibility(self=Get SlDisplay(), InVisibility=Visible)
   > Set Text(Text="LOCTEXT(\"Sound Library\\r\\nAdded\")", self=Get FinishedText())
   > Set Text(Text="LOCTEXT(\"SlName\")", self=Get SLName())
   > SetVisibility(self=Get FinishedAddingBox(), InVisibility=Visible)
   Event Construct()
   > SetVisibility(self=Get FinishedAddingBox(), InVisibility=Hidden)
   SetVisibility(execute=@n10.then, self=Get FinishedAddingBox(), InVisibility=Hidden)
  ]}]
variables [
 CharacterParts {type="array<CustomCharacterPart>", default []}
 KicksItemDef {type=CosmeticShoesItemDefinition, default=null}
 KicksOverrideType {type=E_KicksAudioOverrideType, default=None}
 SoundLibrary {type=SoundLibrary, default=null}
 Timer {type=TimerHandle}
]
widgets {
 info {rootWidgetClass=&StackBox, widgetCount=13}
 widgets [
  StackBox_0 {widgetClassPath=&StackBox, widget=StackBox_0}
  VerticalBox_0 {
   widgetClassPath=&VerticalBox
   parent=StackBox_0
   widget=VerticalBox_0
   slot {class=StackBoxSlot, properties {verticalAlignment=VAlign_Top}}
  }
  HorizontalBox_63 {
   widgetClassPath=&HorizontalBox
   parent=VerticalBox_0
   widget=HorizontalBox_63
   slot {class=VerticalBoxSlot, properties {horizontalAlignment=HAlign_Left}}
  }
  KicksSelection {
   widgetClassPath=&SinglePropertyView
   parent=HorizontalBox_63
   widget=KicksSelection
   properties {propertyName=KicksItemDef}
   slot {class=HorizontalBoxSlot, properties {verticalAlignment=VAlign_Top}}
  }
  SoundLibrarySelection {
   widgetClassPath=&SinglePropertyView
   parent=HorizontalBox_63
   widget=SoundLibrarySelection
   properties {propertyName=KicksOverrideType}
   slot {class=HorizontalBoxSlot, properties {size {sizeRule=Fill, value=1}}}
  }
  AddLibrary {
   widgetClassPath=&EditorUtilityButton
   parent=HorizontalBox_63
   widget=AddLibrary
   properties {widgetStyle {
     clickedSlateSound {resourceObject=null}
     disabled {
      bIsDynamicallyLoaded=false
      drawAs=RoundedBox
      imageType=NoImage
      mirroring=NoMirror
      resourceName=None
      resourceObject=null
      tiling=NoTile
      imageSize=<0 0>
      margin=<ltrb 0 0 0 0>
      outlineSettings {
       bUseBrushTransparency=false
       roundingType=FixedRadius
       width=1
       color {colorUseRule=UseColor_ColorTable, specifiedColor=<rgba 0 0 0 0>}
       cornerRadii=<4 4 4 4>
      }
      tintColor {colorUseRule=UseColor_Specified, specifiedColor=<rgba 0.0395 0.0395 0.0395 1>}
      uVRegion {bIsValid=false, max=<0 0>, min=<0 0>}
     }
     disabledForeground {colorUseRule=UseColor_ColorTable, specifiedColor=<rgba 0 0 0 0>}
     hovered {
      bIsDynamicallyLoaded=false
      drawAs=RoundedBox
      imageType=FullColor
      mirroring=NoMirror
      resourceName=None
      resourceObject=/Game/UI/Foundation/Modals/ErrorWindow/Materials/Icon-Llama-NoEyes
      tiling=NoTile
      imageSize=<256 256>
      margin=<ltrb 0 0 0 0>
      outlineSettings {
       bUseBrushTransparency=false
       roundingType=FixedRadius
       width=1
       color {colorUseRule=UseColor_ColorTable, specifiedColor=<rgba 0 0 0 0>}
       cornerRadii=<4 4 4 4>
      }
      tintColor {colorUseRule=UseColor_Specified, specifiedColor=<rgba 0.0953 0.0953 0.0953 1>}
      uVRegion {bIsValid=false, max=<0 0>, min=<0 0>}
     }
     hoveredForeground {colorUseRule=UseColor_ColorTable, specifiedColor=<rgba 0 0 0 0>}
     hoveredSlateSound {resourceObject=null}
     normal {
      bIsDynamicallyLoaded=false
      drawAs=RoundedBox
      imageType=FullColor
      mirroring=NoMirror
      resourceName=None
      resourceObject=/Game/Athena/SupplyDrops/Llama/AthenaSupplyDropIconLlamaMapIcon
      tiling=NoTile
      imageSize=<36 36>
      margin=<ltrb 0 0 0 0>
      outlineSettings {
       bUseBrushTransparency=false
       roundingType=FixedRadius
       width=1
       color {colorUseRule=UseColor_ColorTable, specifiedColor=<rgba 0 0 0 0>}
       cornerRadii=<4 4 4 4>
      }
      tintColor {colorUseRule=UseColor_Specified, specifiedColor=<rgba 0.0395 0.0395 0.0395 1>}
      uVRegion {bIsValid=false, max=<0 0>, min=<0 0>}
     }
     normalForeground {colorUseRule=UseColor_ColorTable, specifiedColor=<rgba 0 0 0 0>}
     normalPadding=<ltrb 12 1.5 12 1.5>
     pressed {
      bIsDynamicallyLoaded=false
      drawAs=RoundedBox
      imageType=FullColor
      mirroring=NoMirror
      resourceName=None
      resourceObject=null
      tiling=NoTile
      imageSize=<256 256>
      margin=<ltrb 0 0 0 0>
      outlineSettings {
       bUseBrushTransparency=false
       roundingType=FixedRadius
       width=1
       color {colorUseRule=UseColor_ColorTable, specifiedColor=<rgba 0 0 0 0>}
       cornerRadii=<4 4 4 4>
      }
      tintColor {colorUseRule=UseColor_Specified, specifiedColor=<rgba 0.1094 0.1094 0.1094 1>}
      uVRegion {bIsValid=false, max=<0 0>, min=<0 0>}
     }
     pressedForeground {colorUseRule=UseColor_ColorTable, specifiedColor=<rgba 0 0 0 0>}
     pressedPadding=<ltrb 12 2.5 12 0.5>
     pressedSlateSound {resourceObject=null}
    }}
   slot {class=HorizontalBoxSlot}
  }
  TextBlock_277 {
   widgetClassPath=&TextBlock
   parent=AddLibrary
   widget=TextBlock_277
   properties {
    text=Add Library
    colorAndOpacity {colorUseRule=UseColor_Specified, specifiedColor=<rgba 0.5349 0.5468 0.5521 1>}
    font {
     bForceMonospaced=false
     bMaterialIsStencil=false
     fontMaterial=null
     fontObject=&Burbank
     letterSpacing=0
     monospacedWidth=1
     size=9
     skewAmount=0
     typefaceFontName=Burbank-Black
     outlineSettings {
      bApplyOutlineToDropShadows=false
      bMiteredCorners=false
      bSeparateFillAlpha=false
      outlineMaterial=null
      outlineSize=0
      outlineColor=<rgba 0 0 0 1>
     }
    }
   }
   slot {class=ButtonSlot}
  }
  FinishedAddingBox {
   widgetClassPath=&HorizontalBox
   bIsVariable=true
   parent=VerticalBox_0
   widget=FinishedAddingBox
   properties {visibility=Visible}
   slot {class=VerticalBoxSlot, properties {horizontalAlignment=HAlign_Left}}
  }
  FinishedText {
   widgetClassPath=&TextBlock
   bIsVariable=true
   parent=FinishedAddingBox
   widget=FinishedText
   properties {
    justification=Center
    text="Sound Library \r\nadded"
    font {
     bForceMonospaced=false
     bMaterialIsStencil=false
     fontMaterial=null
     fontObject=&Burbank
     letterSpacing=0
     monospacedWidth=1
     size=12.75
     skewAmount=0
     typefaceFontName=Burbank-Black
     outlineSettings {
      bApplyOutlineToDropShadows=false
      bMiteredCorners=false
      bSeparateFillAlpha=false
      outlineMaterial=null
      outlineSize=0
      outlineColor=<rgba 0 0 0 1>
     }
    }
   }
   slot {class=HorizontalBoxSlot, properties {horizontalAlignment=HAlign_Center, verticalAlignment=VAlign_Center}}
  }
  SlDisplay {widgetClassPath=&VerticalBox, bIsVariable=true, parent=FinishedAddingBox, widget=SlDisplay, slot {class=HorizontalBoxSlot}}
  SizeBox_1 {
   widgetClassPath=&SizeBox
   parent=SlDisplay
   widget=SizeBox_1
   properties {bOverride_WidthOverride=true, widthOverride=95.5874}
   slot {class=VerticalBoxSlot}
  }
  SoundLibraryAdded {
   widgetClassPath=&SinglePropertyView
   parent=SizeBox_1
   widget=SoundLibraryAdded
   properties {nameOverride=" ", propertyName=SoundLibrary, visibility=HitTestInvisible}
   slot {class=SizeBoxSlot, properties {horizontalAlignment=HAlign_Center, verticalAlignment=VAlign_Center, padding=<ltrb 5 0 0 0>}}
  }
  SLName {
   widgetClassPath=&TextBlock
   bIsVariable=true
   parent=SlDisplay
   widget=SLName
   properties {
    text=Text Block
    font {
     bForceMonospaced=false
     bMaterialIsStencil=false
     fontMaterial=null
     fontObject=&Burbank
     letterSpacing=0
     monospacedWidth=1
     size=9
     skewAmount=0
     typefaceFontName=Burbank-Black
     outlineSettings {
      bApplyOutlineToDropShadows=false
      bMiteredCorners=false
      bSeparateFillAlpha=false
      outlineMaterial=null
      outlineSize=0
      outlineColor=<rgba 0 0 0 1>
     }
    }
   }
   slot {class=VerticalBoxSlot, properties {horizontalAlignment=HAlign_Center}}
  }
 ]
}
